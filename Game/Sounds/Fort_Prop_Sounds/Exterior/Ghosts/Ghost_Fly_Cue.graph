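# Sound cue for ghost flying sounds, part of the exterior prop sound set.
# A Random node picks one of seven equally weighted Ghost_Fly waves, which feeds a Modulator that randomly raises the pitch slightly.
# The cue overrides attenuation for distance falloff and stereo spread, plays in the in-game SFX class at reduced volume, and caps concurrency at a few voices, stopping the oldest.
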SoundCue /Game/Sounds/Fort_Prop_Sounds/Exterior/Ghosts/Ghost_Fly_Cue
root=Modulator_1
nodes [
 {id=Modulator_1, children [Random_0], properties {pitchMax=1.3, pitchMin=1.1}}
 {
  id=Random_0
  children [WavePlayer_1,WavePlayer_4,WavePlayer_5,WavePlayer_6,WavePlayer_7,WavePlayer_8,WavePlayer_9]
  properties {weights [1,1,1,1,1,1,1]}
 }
 {id=WavePlayer_1, properties {soundWaveAssetPtr=&/Ghost_Fly_01}}
 {id=WavePlayer_4, properties {soundWaveAssetPtr=&/Ghost_Fly_02}}
 {id=WavePlayer_5, properties {soundWaveAssetPtr=&/Ghost_Fly_03}}
 {id=WavePlayer_6, properties {soundWaveAssetPtr=&/Ghost_Fly_04}}
 {id=WavePlayer_7, properties {soundWaveAssetPtr=&/Ghost_Fly_05}}
 {id=WavePlayer_8, properties {soundWaveAssetPtr=&/Ghost_Fly_06}}
 {id=WavePlayer_9, properties {soundWaveAssetPtr=&/Ghost_Fly_07}}
]
properties {
 bOverrideAttenuation=true
 bOverrideConcurrency=true
 maxDistance=5000
 soundClassObject=&Fort_SFX_InGame
 volumeMultiplier=0.8
 attenuationOverrides {falloffDistance=4000, stereoSpread=600, attenuationShapeExtents {x=1000}}
 concurrencyOverrides {maxCount=4, resolutionRule=StopOldest}
}
